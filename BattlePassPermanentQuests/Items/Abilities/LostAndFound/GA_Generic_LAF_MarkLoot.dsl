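; Gameplay ability for the Lost and Found quest item that spawns a loot spawner near the player to handle spawning and marking loot.
; On activation the server waits on a short timer, spawns BGA_Generic_LAF_LootSpawner ahead of and offset vertically from the avatar pawn, then ends.
; CheckForRarity (tag-based pickup search) and MarkRarestItem (server-side marker) are defined but not called in this graph.
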(event Ability|EventActivateAbility
  ;[0 0 0] Spawn a replicated but invisible actor to handle the spawning and marking of loot after a given delay
  (Ability|HasAuthority
    (:True
      (Utilities|Time|SetTimerbyEvent (delegate Custom|Delay) 2.0 false true))))

(event Custom|Delay
  ;[0 0 0] Spawn a replicated but invisible actor to handle the spawning and marking of loot after a given delay
  (Game|SpawnActorfromClass "/BattlePassPermanentQuests/Blueprints/BGA_Generic_LAF_LootSpawner.BGA_Generic_LAF_LootSpawner_C" (Math|Transform|MakeTransform (+ (Math|Vector|MakeVector (.x (Transformation|GetActorLocation (CallFunction|GetAvatarasFortPlayerPawn))) (.y (Transformation|GetActorLocation (CallFunction|GetAvatarasFortPlayerPawn))) (+ (.z (Transformation|GetActorLocation (CallFunction|GetAvatarasFortPlayerPawn))) (Variables|Default|GetDistanceFromGround))) (* (Variables|Default|GetDistanceFromPlayer) (Transformation|GetActorForwardVector (CallFunction|GetAvatarasFortPlayerPawn))))) "Undefined" "MultiplyWithRoot" (CallFunction|GetAvatarasFortPlayerPawn) 0 0)
  (Ability|EndAbility))

(fn CheckForRarity
  (params (Items :type Array%20of%20Actor%20Object%20References) (Rarity :type Gameplay%20Tag%20Structure%20%28by%20ref%29))
  (for _array_element Items
    (bind _asfort_pickup_athena (Utilities|Casting|CastToFortPickupAthena _array_element))
    (if (GameplayTags|HasTag (GameplayTags|GetOwnedGameplayTags _asfort_pickup_athena :!class /Script/GameplayTags.BlueprintGameplayTagLibrary) Rarity)
      (Variables|Default|SetRarest%20Item _asfort_pickup_athena)
      (Variables|Default|SetDone%20Searching true)
      (Development|PrintString (Utilities|String|Append "LOOTY FOUND RARE ITEM." "\\R\\N LOOTY RARITY = " (GameplayTags|GetDebugStringfromGameplayTag Rarity) (Utilities|GetDisplayName _asfort_pickup_athena))))))

(fn MarkRarestItem
  (MarkerComponent|MarkActoronServer (Class|FortPlayerController|GetMarkerComponent (Fortnite|GetFortPlayerControllerfromActor (Ability|GetOwningActorfromActorInfo))) (Variables|Default|GetRarest%20Item) true true)
  (Development|PrintString "LOOTY MARKED SOMETHING")
  (Ability|EndAbility))
