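; Gameplay ability for a Lost and Found (LAF) quest that reports when the player has a ranged weapon equipped.
; On the server it waits briefly for quest initialization, then checks the current weapon's tags or listens for weapon-equipped events until one matches.
; It broadcasts the LAF equip custom verb with the filtered weapon tags and ends.
; The rarity-check and item-marking functions are not called in this graph and look like leftovers.
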(event Ability|EventActivateAbility
  (bind _returnvalue (CallFunction|GetAvatarasFortPlayerPawn))
  (Ability|HasAuthority
    (:True
      (Utilities|IsValid _returnvalue
        (:Is%20Valid
          ; Insert Slight delay to ensure this happens after the quest has initialized. Also gives player time to switch weapons
          (Utilities|FlowControl|Delay 2.0)
          ; If the Player already has a weapon equipped, fire the Verb and end the ability.
          (bind _output_get (Variables|Default|SetPlayer _returnvalue))
          (bind _output_get_2 (Variables|Default|SetWeaponTags (Tags|GetWeaponDataGameplayTags (Weapon|GetCurrentWeapon _output_get))))
          (if (GameplayTags|HasAnyTags _output_get_2 (Variables|Default|GetAnyRangedWeapon))
            (CallFunction|BroadcastEquipVerb (Variables|Default|GetPlayer))
            (else
              ; If the Player doesn't have a weapon equipped, wait until they do and then fire the Verb.
              (FortPawn|BindEventtoOnWeaponEquippedDelegate (Variables|Default|GetPlayer) (delegate Custom|WeaponEquipped)))))
        (:Is%20Not%20Valid
          (Ability|EndAbility))))))

(event Custom|WeaponEquipped
  (params (NewWeapon :type Fort%20Weapon%20Object%20Reference) (PrevWeapon :type Fort%20Weapon%20Object%20Reference))
  (bind _gameplaytags (Tags|GetWeaponDataGameplayTags NewWeapon))
  (bind _player (Variables|Default|GetPlayer))
  ; If the Player doesn't have a weapon equipped, wait until they do and then fire the Verb.
  (if (GameplayTags|HasAnyTags _gameplaytags (Variables|Default|GetAnyRangedWeapon))
    (Variables|Default|SetWeaponTags _gameplaytags)
    (FortPawn|UnbindEventfromOnWeaponEquippedDelegate _player (delegate Custom|WeaponEquipped))
    (CallFunction|BroadcastEquipVerb _player)))

(fn BroadcastEquipVerb
  (params (Author :type Object%20Reference%20%28by%20ref%29))
  ; Broadcast verb and send along the weapon tags, filtered to what we check against
  (bind _result (CustomVerb|AddTagsforSubject (CustomVerb|MakeCustomVerbMessage Author "(TagName=\"CustomVerb.LAF.Equip\")") "Weapon" (GameplayTags|Filter (Variables|Default|GetWeaponTags) (Variables|Default|GetAnyRangedWeapon))))
  (CustomVerb|BroadcastCustomVerbMessage _result)
  (Ability|EndAbility))

(fn CheckForRarity
  (params (Items :type Array%20of%20Actor%20Object%20References) (Rarity :type Gameplay%20Tag%20Structure%20%28by%20ref%29))
  (for _array_element Items
    (bind _asfort_pickup_athena (Utilities|Casting|CastToFortPickupAthena _array_element))
    (if (GameplayTags|HasTag (GameplayTags|GetOwnedGameplayTags _asfort_pickup_athena :!class /Script/GameplayTags.BlueprintGameplayTagLibrary) Rarity)
      _)))

(fn MarkRarestItem
  (MarkerComponent|MarkActoronServer (Class|FortPlayerController|GetMarkerComponent (Fortnite|GetFortPlayerControllerfromActor (Ability|GetOwningActorfromActorInfo))) 0 true true)
  (Development|PrintString "LOOTY MARKED SOMETHING")
  (Ability|EndAbility))
